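# Unlit, translucent, two-sided Fresnel sphere effect material for Niagara mesh particles, filed under Battle Pass S25 FX.
# An inverted DepthFade forms a mask that is strongest where the mesh nears opaque scene geometry.
# That mask separately multiplies a Fresnel rim term and particle color alpha.
# The graph shows no links from these products or a lone constant to the material outputs.
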Material /BattlePassS25/FX/Part3/Materials/M_FresnelSphere_S25_Q2
graph {
 properties {
  bUsedWithNiagaraMeshParticles=true
  blendMode=BLEND_Translucent
  pixelDepthOffsetMode=PDOM_Legacy
  shadingModel=MSM_Unlit
  twoSided=true
  usedShadingModels=Unlit
 }
 nodes [
  Constant() {r=1}
  n1 = OneMinus(None=DepthFade() {fadeDistanceDefault=5})
  Multiply(A=Fresnel() {baseReflectFraction=0, exponent=2}, B=@n1)
  Multiply(A=ParticleColor().A, B=@n1)
 ]
}
