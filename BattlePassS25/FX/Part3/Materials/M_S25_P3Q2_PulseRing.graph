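# Translucent unlit Niagara sprite and mesh particle material for a Season 25 Battle Pass pulse ring effect.
# A plasma texture scrolled in quantized time steps distorts U coordinates feeding a soft SphereMask band.
# That band drives a curve-atlas blend between background and highlight colors, tinted by particle color.
# It also shapes opacity with particle alpha, panning Perlin noise and an optional V-edge fade.
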Material /BattlePassS25/FX/Part3/Materials/M_S25_P3Q2_PulseRing
graph {
 properties {
  bUsedWithNiagaraMeshParticles=true
  bUsedWithNiagaraSprites=true
  blendMode=BLEND_Translucent
  pixelDepthOffsetMode=PDOM_Legacy
  shadingModel=MSM_Unlit
  usedShadingModels=Unlit
 }
 nodes [
  n1 = VectorParameter() {parameterName=ColorBackground, defaultValue=<rgba 0 0.3333 2 1>}
  n2 = VectorParameter() {parameterName=ColorHighlight, defaultValue=<rgba 0.15 1.575 3 1>}
  n3 = MaterialFunctionCall(Float2=Add(A=TextureCoordinate(), B=ParticleRandom())) {materialFunction=&BreakOutFloat2Components}
  n4 = Multiply(A=@n3.R, B=ScalarParameter() {defaultValue=0.01, parameterName=Distort_Scale_U}) {constB=0.01}
  n5 = Multiply(A=@n3.G, B=ScalarParameter() {defaultValue=1, parameterName=Distort_Scale_V})
  n6 = Multiply(A=Time(), B=ScalarParameter() {defaultValue=1, parameterName=Speed X, sliderMax=10})
  n7 = ScalarParameter() {defaultValue=6, parameterName=Steps, sliderMax=50}
  n8 = Divide(A=Ceil(None=Multiply(A=@n6, B=@n7)), B=@n7)
  n9 = Constant()
  n10 = TextureSampleParameter2D(UVs=Add(A=AppendVector(A=@n4, B=@n5), B=AppendVector(A=@n8, B=@n9))) {
   parameterName=Distort_Tex
   samplerType=SAMPLERTYPE_LinearGrayscale
   texture=/Game/Effects/Fort_Effects/Emotes/UltraLight/Texture/T_Plasma_B_Greyscale
  }
  n11 = ScalarParameter() {defaultValue=0.2, parameterName=Distort_Intensity}
  n12 = MaterialFunctionCall(Float2=Add(A=Multiply(A=@n10.R, B=@n11) {constB=0.2}, B=TextureCoordinate())) {materialFunction=&BreakOutFloat2Components}
  n13 = Constant() {r=1}
  n14 = TextureCoordinate()
  n15 = Saturate(None=SphereMask(A=ComponentMask(None=@n14) {g=true}, B=Constant() {r=0.5}) {attenuationRadius=0.55, hardnessPercent=10})
  n16 = SphereMask(
   A=MaterialFunctionCall(UVs=Add(A=@n12.R) {constB=-0.1}, Texture Scale=AppendVector(A=@n13, B=@n15)) {materialFunction=&ScaleUVsByCenter}.UVs
   B=Constant() {r=0.5}
   Radius=ScalarParameter() {defaultValue=0.2, parameterName=Gradient_Radius}
   Hardness=ScalarParameter() {parameterName=Gradient_Hardness}
  ) {attenuationRadius=0.2, hardnessPercent=20}
  n17 = Saturate(None=@n16)
  n18 = CurveAtlasRowParameter(CurveTime=@n17) {
   atlas=/Game/Athena/LinearColorCurves/VFX_CurveAtlas
   curve=/Game/Gadgets/Assets/VinderTech_GliderChute/Glider_Majesty_Male/FX/Curve_Majesty_Ramp
   defaultValue=25
   parameterName=Curve
  }
  n19 = ParticleColor()
  MaterialFunctionCall(Emissive=Multiply(A=LinearInterpolate(A=@n1.RGB, B=@n2.RGB, Alpha=@n18), B=@n19.RGB)) {materialFunction=&MF_FX_EmissiveAdjustments}
  n20 = MaterialFunctionCall(Float2=TextureCoordinate()) {materialFunction=&BreakOutFloat2Components}
  n21 = Multiply(A=@n20.R) {constB=0.1}
  n22 = Multiply(A=@n20.G) {constB=3}
  n23 = ScalarParameter() {defaultValue=1, parameterName=Noise_Pan_U_Speed}
  n24 = Constant()
  n25 = TextureSampleParameter2D(UVs=Panner(Coordinate=AppendVector(A=@n21, B=@n22), Speed=AppendVector(A=@n23, B=@n24)) {speedX=1}) {
   parameterName=TexNoise
   samplerType=SAMPLERTYPE_LinearGrayscale
   texture=&T_SimplePerlin_Greyscale
  }
  n26 = LinearInterpolate(Alpha=@n25.G) {constA=0.2}
  n27 = ScalarParameter() {defaultValue=1, parameterName=Noise_Intensity}
  n28 = Constant() {r=0.5}
  n29 = Multiply(
   A=Power(
    Base=SphereMask(A=@n20.G, B=@n28) {attenuationRadius=0.5, hardnessPercent=0}
    Exp=ScalarParameter() {defaultValue=1, parameterName=Fade_Edges_Power}
   )
   B=ScalarParameter() {defaultValue=1, parameterName=Fade_Edges_Hardness}
  )
  Multiply(
   A=Multiply(A=Multiply(A=@n19.A, B=@n16), B=LinearInterpolate(B=@n26, Alpha=@n27) {constA=1})
   B=LinearInterpolate(A=Constant() {r=1}, B=Saturate(None=@n29), Alpha=ScalarParameter() {parameterName=Fade_Edges})
  )
  n30 = Power(Base=@n16) {constExponent=1}
  Saturate(None=LinearInterpolate(A=@n17, B=Multiply(A=@n30)) {constAlpha=0.8})
 ]
}
