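# Holographic material for the Ancients telescope prop in Battle Pass S25 FX, built on the shared MF_Holo function with the telescope's diffuse and normal textures.
# World position offset combines the holo offset run through vertex-color element hiding, an HT_CrunchVerts pull toward object position, and slow time-driven rotation about the pivot.
# Normals are corrected for that rotation.
# Holo emissive is scaled by time of day, dimming at night.
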Material /BattlePassS25/FX/Part3/Materials/M_TelescopeShader_Holo
graph {
 properties {
  bUseMaterialAttributes=true
  bUsedWithClothing=true
  bUsedWithMeshDeformer=true
  bUsedWithMeshParticles=true
  bUsedWithMorphTargets=true
  bUsedWithSkeletalMesh=true
  bUsedWithStaticLighting=true
  numCustomizedUVs=3
  physMaterial=/Game/FortressPhysicalMaterials/Player_Flesh
  pixelDepthOffsetMode=PDOM_Legacy
  shadingModel=MSM_SubsurfaceProfile
  subsurfaceProfile=/Game/Packages/Fortress_SharedMaterials/SubSurfaceProfile/Character_SubsurfaceProfile
  usedShadingModels=Subsurface Profile
 }
 nodes [
  n1 = TextureObjectParameter() {
   parameterName=Normals
   samplerType=SAMPLERTYPE_Normal
   texture=/Game/Environments/Asteria/Props/Ancients/Ancients_Telescope_A/Textures/T_Ancients_Telescope_A_A_N
  }
  n2 = GetMaterialAttributes(None=MaterialFunctionCall!(
    BaseColor=TextureObjectParameter() {parameterName=Diffuse, texture=/Game/Environments/Asteria/Props/Ancients/Ancients_Telescope_A/Textures/T_Ancients_Telescope_A_A_D}
    Normal=@n1
   ) {materialFunction=/Game/UI/Frontend/Lobby/HeroSquad/MF_Holo}.Result) {attributeGetTypes [#1, #2]}
  n3 = MaterialFunctionCall!(WPO=@n2.World Position Offset) {materialFunction=/Game/Packages/Fortress_SharedMaterials/Base_Material_Functions/Fortnite_HideElementsWithVertexColor}
  n4 = Multiply(A=Subtract(A=ObjectPositionWS(), B=WorldPosition().XYZ), B=ScalarParameter() {parameterName=HT_CrunchVerts})
  n5 = Normalize(VectorInput=MaterialFunctionCall!(VectorToRotate=ObjectOrientation(), LookAtVector=Constant3Vector() {constant=<rgba 0 0 1 0>}, RestingVector) {materialFunction=/Engine/Functions/Engine_MaterialFunctions01/Vectors/RotateVector}.Result)
  n6 = Multiply(A=Time()) {constB=-0.05}
  n7 = MaterialFunctionCall() {materialFunction=&ObjectPivotPoint}
  n8 = WorldPosition()
  n9 = TextureSampleParameter2D() {
   parameterName=Normal
   samplerType=SAMPLERTYPE_Normal
   texture=&T_Ancients_Telescope_A_A_N
  }
  n10 = Constant() {r=1}
  n11 = Constant() {r=0.2}
  SetMaterialAttributes!(
   MaterialAttributes=@n2.MaterialAttributes
   World Position Offset=Add(
    A=Add(A=@n3.Result, B=@n4)
    B=RotateAboutAxis(NormalizedRotationAxis=@n5, RotationAngle=@n6, PivotPoint=@n7.Object Pivot Location, Position=@n8.XYZ)
   )
   Normal=MaterialFunctionCall(Rotation Axis=@n5, Rotation Angle=@n6, World Space Vertex Normals=Transform(None=@n9.RGB)) {materialFunction=&FixRotateAboutAxisNormals}.New Tangent Space Vertex Normal
   Emissive Color=Multiply(
    A=@n2.Emissive Color
    B=MaterialFunctionCall(Value_Morning=@n10, Value_Day=@n10, Value_Evening=@n10, Value_Night=@n11) {materialFunction=&Scalar_Time_of_Day}.Result
   )
  ) {attributeSetTypes [#1, #3, #2]}
 ]
}
